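# Physics asset for the Critter Maniac Male two-handed pickaxe, defining bodies and joints for its dangling tape strips.
# A capsule on Melee_TwoHanded_Handle anchors four chains of dyn_tape capsules: one with three segments and three with two.
# Each segment is constrained to the one before it or to the handle, with per-joint swing and twist limits.
# The chain 3 capsules and dyn_tape_4_2 all reuse the shape name dyn_tape_4_1_capsule.
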PhysicsAsset /BRCosmetics/Weapons/FORT_Melee/Pickaxe_Critter_Maniac_Male/Meshes/Critter_Maniac_Male_Axe_Physics
bodies {
 Melee_TwoHanded_Handle [{shapeType=Capsule, length=12.2826, radius=2.7911}]
 dyn_tape_1_1 [{shapeType=Capsule, length=5.0144, radius=2.3768, shapeName=dyn_tape_1_1_capsule, center=<6.5 0.0001 0>, rotation=<rot 90 0 0>}]
 dyn_tape_1_2 [{shapeType=Capsule, length=4.3151, radius=2.2308, shapeName=dyn_tape_1_2_capsule, center=<4.5 0 0>, rotation=<rot 90 0 0>}]
 dyn_tape_1_3 [{shapeType=Capsule, length=1.2636, radius=2, shapeName=dyn_tape_1_3_capsule, center=<2.5 0 0>, rotation=<rot 90 0 0>}]
 dyn_tape_2_1 [{shapeType=Capsule, length=1.5, radius=1, shapeName=dyn_tape_2_1_capsule, center=<1.7167 0 0>, rotation=<rot 90 0 0>}]
 dyn_tape_2_2 [{shapeType=Capsule, length=2, radius=1, shapeName=dyn_tape_2_2_capsule, center=<1.5067 0 0>, rotation=<rot 90 0 0>}]
 dyn_tape_3_1 [{shapeType=Capsule, length=2, radius=1, shapeName=dyn_tape_4_1_capsule, center=<1.7631 0 0>, rotation=<rot 90 0 0>}]
 dyn_tape_3_2 [{shapeType=Capsule, length=2, radius=1, shapeName=dyn_tape_4_1_capsule, center=<2.0626 0 0>, rotation=<rot 90 0 0>}]
 dyn_tape_4_1 [{shapeType=Capsule, length=2, radius=1, shapeName=dyn_tape_4_1_capsule, center=<1.7631 0 0>, rotation=<rot 90 0 0>}]
 dyn_tape_4_2 [{shapeType=Capsule, length=2, radius=1, shapeName=dyn_tape_4_1_capsule, center=<2.0626 0 0>, rotation=<rot 90 0 0>}]
}
constraints [
 {bone1Name=dyn_tape_1_1, bone2Name=Melee_TwoHanded_Handle, swing1LimitDegrees=65, swing2LimitDegrees=65, twistLimitDegrees=20}
 {bone1Name=dyn_tape_1_2, bone2Name=dyn_tape_1_1, swing1LimitDegrees=45, swing2LimitDegrees=45, twistLimitDegrees=20}
 {bone1Name=dyn_tape_1_3, bone2Name=dyn_tape_1_2, swing1LimitDegrees=50, swing2LimitDegrees=50, twistLimitDegrees=10}
 {bone1Name=dyn_tape_2_1, bone2Name=Melee_TwoHanded_Handle, swing1LimitDegrees=45, swing2LimitDegrees=25, twistLimitDegrees=10}
 {bone1Name=dyn_tape_2_2, bone2Name=dyn_tape_2_1, swing1LimitDegrees=45, swing2LimitDegrees=45, twistLimitDegrees=5}
 {bone1Name=dyn_tape_3_1, bone2Name=Melee_TwoHanded_Handle, swing1LimitDegrees=50, swing2LimitDegrees=35, twistLimitDegrees=45}
 {bone1Name=dyn_tape_3_2, bone2Name=dyn_tape_3_1, swing1LimitDegrees=45, swing2LimitDegrees=45, twistLimitDegrees=45}
 {bone1Name=dyn_tape_4_1, bone2Name=Melee_TwoHanded_Handle, swing1LimitDegrees=45, swing2LimitDegrees=25, twistLimitDegrees=5}
 {bone1Name=dyn_tape_4_2, bone2Name=dyn_tape_4_1, swing1LimitDegrees=45, swing2LimitDegrees=45, twistLimitDegrees=5}
]
