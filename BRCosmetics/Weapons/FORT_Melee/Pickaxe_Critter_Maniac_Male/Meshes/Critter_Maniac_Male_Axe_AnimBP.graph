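# Animation Blueprint for the Critter Maniac Male pickaxe axe that adds secondary motion to its dangling tape bones.
# The input pose gets an additive rotation on tape 1, then runs through a Trail controller with fake velocity and a RigidBody simulation in handle-bone space.
# Each update reads bIsInFrontEnd to switch the tape 1 rotation and enable post-simulation yaw offsets on tapes 3 and 4 only in the front end.
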AnimBlueprint /BRCosmetics/Weapons/FORT_Melee/Pickaxe_Critter_Maniac_Male/Meshes/Critter_Maniac_Male_Axe_AnimBP
target_skeleton=&/Critter_Maniac_Male_Axe_Skeleton
graphs [
 AnimGraph {nodes [
   AnimGraphNode_SubInput_0 = Input Pose()
   n1 = "Transform (Modify) Bone - Bone: dyn_tape_1_1"(
    ComponentPose=Local To Component(LocalPose=@AnimGraphNode_SubInput_0)
    Alpha=1
    Translation="0,0,0"
    Rotation=Get frontendRotate() -> (frontendRotate:Rotator)
    Scale="1,1,1"
   ) {
    rotationMode=BMM_Additive
    rotationSpace=BCS_BoneSpace
    alphaBoolBlend {blendOption=Linear}
    boneToModify {boneName=dyn_tape_1_1}
    rotation=<rot 0.712 30.852 0.43>
   }
   n2 = Trail controller - Bone: dyn_tape_1_3(ComponentPose=@n1, Alpha=1) {
    bActorSpaceFakeVel=true
    chainLength=3
    baseJoint {boneName=Melee_TwoHanded_Handle}
    fakeVelocity=<0.3 0.3 0.3>
    rotationLimits [{limitMax=<180 180 180>}, {limitMax=<180 180 180>}, {limitMax=<180 180 180>}]
    rotationOffsets [<0 0 0>, <0 0 0>, <0 0 0>]
    trailBone {boneName=dyn_tape_1_3}
   }
   n3 = RigidBody!(ComponentPose=@n2, Alpha, OverrideWorldGravity:Vector="0,0,-980", ExternalForce) -> (Pose) {
    bClampLinearTranslationLimitToRefPose=true
    bOverrideWorldGravity=true
    overridePhysicsAsset=&/Critter_Maniac_Male_Axe_Physics
    simulationSpace=BaseBoneSpace
    alphaBoolBlend {blendOption=Linear}
    baseBoneRef {boneName=Melee_TwoHanded_Handle}
    componentAppliedLinearAccClamp=<4000 4000 4000>
    componentLinearAccScale=<0.4 0.3 0.3>
    componentLinearVelScale=<4 0.3 0.3>
   }
   n4 = Get Is in Front End() -> (Is in Front End:double)
   n5 = "Transform (Modify) Bone - Bone: dyn_tape_3_1"(ComponentPose=@n3, Alpha=@n4, Translation="0,0,0", Rotation="0,10,0", Scale="1,1,1") {
    rotationMode=BMM_Additive
    rotationSpace=BCS_BoneSpace
    alphaBoolBlend {blendOption=Linear}
    boneToModify {boneName=dyn_tape_3_1}
    rotation {yaw=10}
   }
   n6 = "Transform (Modify) Bone - Bone: dyn_tape_4_1"(ComponentPose=@n5, Alpha=@n4, Translation="0,0,0", Rotation="0,20,0", Scale="1,1,1") {
    rotationMode=BMM_Additive
    rotationSpace=BCS_BoneSpace
    alphaBoolBlend {blendOption=Linear}
    boneToModify {boneName=dyn_tape_4_1}
    rotation {yaw=20}
   }
   Output Pose(Result=Component To Local(ComponentPose=@n6)) {layerGroup=None}
  ]}
 EventGraph {nodes [
   Event BlueprintUpdateAnimation()
   > n1 = Sequence()
   > Set frontendRotate(frontendRotate:Rotator=SelectRotator!(A:Rotator="30,0, 0", B:Rotator="0.712,30.852,0.43", bPickA:bool=Get bIsInFrontEnd()) -> (ReturnValue:Rotator)) -> (Output_Get:Rotator)
   Set Is in Front End(execute=@n1.then_1, Is in Front End:double="To Float (Boolean)"(InBool=Get bIsInFrontEnd())) -> (Output_Get:double)
  ]}
]
variables [
 frontendRotate {type=Rotator}
 Is in Front End {type=number}
]
